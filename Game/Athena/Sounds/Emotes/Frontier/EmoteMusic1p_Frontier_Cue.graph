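# First-person music cue for the Frontier emote in Athena.
# A mixer plays the start and stinger waves together, with the start slightly quieter.
# It uses the first-person emote music sound class and concurrency rule, keeps playing when silent, and has a very large max distance.
# A submix send also routes it to the music audio analysis submix.
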SoundCue /Game/Athena/Sounds/Emotes/Frontier/EmoteMusic1p_Frontier_Cue
root=Mixer_0
nodes [
 {id=Mixer_0, children [WavePlayer_0,WavePlayer_1], properties {inputVolume [0.85,1]}}
 {id=WavePlayer_0, properties {soundWaveAssetPtr=&/Emote_Frontier_Start}}
 {id=WavePlayer_1, properties {soundWaveAssetPtr=&/Emote_Frontier_Stinger}}
]
properties {
 maxDistance=2097152
 soundClassObject=/Game/Sounds/FortSoundClassesAndModes/Fort_Music_Emote_1P_PSM
 virtualizationMode=PlayWhenSilent
 volumeMultiplier=1
 concurrencySet [/Game/Sounds/Concurrency_Rules/Athena_EmoteMusic_1P_Conc]
 soundSubmixSends [{minSendLevel=0, soundSubmix=&Fortnite_AudioAnalysis_Music}]
}
